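# Translucent unlit material for a distant comet effect, built for mesh particles.
# Panning swirl and smooth noise, combined with a vertical UV falloff and second-UV sphere mask, shape opacity and a dark-to-light color gradient.
# A glowing center and squared, tinted dust debris layer add to the color, remapped through an HDR helper.
# Time-of-day functions tint the color and scale particle-alpha-driven opacity, strongest at night and faintest by day.
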Material /Figment_S03/Effects/Fort_Effects/Materials/Comet/M_Comet_Distant
graph {
 properties {bUsedWithMeshParticles=true, blendMode=BLEND_Translucent, pixelDepthOffsetMode=PDOM_Legacy, shadingModel=MSM_Unlit, usedShadingModels=Unlit}
 nodes [
  n1 = Multiply(A=Time(), B=ScalarParameter() {defaultValue=0.5, parameterName=PanSpeed})
  n2 = TextureSample(UVs=Panner(Coordinate=TextureCoordinate() {uTiling=2}, Time=@n1) {speedX=-0.04, speedY=-0.08}) {samplerType=SAMPLERTYPE_LinearGrayscale, texture=/Game/Global/Textures/Default/Noise/T_SwirlNoise_N_height}
  n3 = TextureSample(UVs=Panner(Coordinate=TextureCoordinate(), Time=@n1) {speedX=0.05, speedY=-0.15}) {texture=/Game/Effects/Fort_Effects/Textures/T_Smooth_tile}
  n4 = LinearInterpolate(A=@n2.R, B=@n3.R)
  Add(A=@n4, B=Multiply(A=@n2.R, B=@n3.R))
  n5 = ComponentMask(None=TextureCoordinate()) {g=true}
  n6 = TextureCoordinate() {coordinateIndex=1}
  n7 = Constant() {r=0.5}
  n8 = Add(
   A=Power(Base=OneMinus(None=@n5)) {constExponent=3}
   B=Power(Base=SphereMask(A=@n6, B=@n7) {attenuationRadius=0.22, hardnessPercent=25}) {constExponent=3}
  )
  n9 = ComponentMask(None=TextureCoordinate()) {g=true}
  n10 = Multiply(A=@n8, B=Power(Base=Multiply(A=@n4) {constB=2}, Exp=LinearInterpolate(Alpha=@n9) {constB=8}))
  n11 = Constant() {r=0.4}
  Multiply(
   A=Multiply(A=ParticleColor().A, B=Add(A=@n8, B=@n10))
   B=MaterialFunctionCall(Color_Morning=@n11, Color_Day=Constant() {r=0.12}, Color_Evening=@n11, Color_Night=Constant() {r=1}) {materialFunction=&Color_Time_of_Day}.Result
  )
  n12 = VectorParameter() {parameterName=EveningTint, defaultValue=<rgba 0.8 0.5555 0.4375 0>}
  n13 = Power(Base=SphereMask(A=@n6, B=@n7) {attenuationRadius=0.15, hardnessPercent=0}) {constExponent=4}
  n14 = VectorParameter() {parameterName=CenterColor, defaultValue=<rgba 1.7245 2.3418 3 0>}
  n15 = LinearInterpolate(
   A=VectorParameter() {parameterName=DarkColor, defaultValue=<rgba 0.0757 0.1086 0.2344 0>}.RGB
   B=VectorParameter() {parameterName=LightColor, defaultValue=<rgba 0.0223 0.3654 1 0>}.RGB
   Alpha=@n10
  )
  n16 = TextureSample(UVs=Panner(Coordinate=TextureCoordinate() {vTiling=2}, Time=@n1) {speedX=0.02, speedY=-0.15}) {texture=/Game/Global/Textures/Default/PointSplatter/T_Tiling_Dust}
  n17 = Add(
   A=Add(A=Multiply(A=@n13, B=@n14.RGB), B=Clamp(None=@n15) {maxDefault=10})
   B=Multiply(A=Multiply(A=@n16.RGB, B=@n16.RGB), B=VectorParameter() {parameterName=DebrisColor, defaultValue=<rgba 12 6.75 11.8657 0>}.RGB)
  )
  n18 = ScalarParameter() {defaultValue=2, parameterName=HDRMult}
  Multiply(
   A=MaterialFunctionCall(
    Color_Morning=@n12.RGB
    Color_Day=VectorParameter() {parameterName=DayTint, defaultValue=<rgba 0.175 0.7 0.5361 0>}.RGB
    Color_Evening=@n12.RGB
    Color_Night=VectorParameter() {parameterName=NightTint, defaultValue=<rgba 0.4531 0.166 1 0>}.RGB
   ) {materialFunction=&Color_Time_of_Day}.Result
   B=MaterialFunctionCall(LDR Input=Clamp(None=@n17), HDR Tint=Multiply(A=@n17, B=@n18)) {materialFunction=&DeriveHDRfromLDR}.RemappedRange
  )
 ]
}
