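# Unlit additive particle material for a soft, cloudy glow, apparently surrounding a comet.
# Two storm-cloud and fog textures pan in opposite vertical directions, blend, are lifted by GlowAdd and are shaped by a radial sphere mask.
# The result blends dark to light colors, multiplied by particle color and time-of-day tints, and also scales particle alpha.
# A separate time-of-day factor scales the offset between object and world position.
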Material /Figment_S03/Effects/Fort_Effects/Materials/Comet/M_Comet_Surround
graph {
 properties {
  bUsedWithMeshParticles=true
  bUsedWithParticleSprites=true
  blendMode=BLEND_Additive
  pixelDepthOffsetMode=PDOM_Legacy
  shadingModel=MSM_Unlit
  usedShadingModels=Unlit
 }
 nodes [
  Constant() {r=2}
  n1 = Constant() {r=0.8}
  LinearInterpolate(
   A=Subtract(A=ObjectPositionWS(), B=WorldPosition().XYZ)
   Alpha=MaterialFunctionCall(Color_Morning=@n1, Color_Day=Constant() {r=0.6}, Color_Evening=@n1, Color_Night=Constant() {r=1}) {materialFunction=&Color_Time_of_Day}.Result
  ) {constAlpha=0, constB=0}
  n2 = ParticleColor()
  n3 = Constant() {r=0.5}
  n4 = SphereMask(A=TextureCoordinate(), B=@n3, Radius=@n3) {attenuationRadius=0.5, hardnessPercent=25}
  n5 = Multiply(A=Time(), B=ScalarParameter() {defaultValue=0.3, parameterName=PanSpeed})
  n6 = TextureSample(UVs=Panner(Coordinate=TextureCoordinate() {uTiling=1.25, vTiling=1.25}, Time=@n5) {speedX=0.01, speedY=-0.05}) {samplerType=SAMPLERTYPE_LinearGrayscale, texture=/Game/VisualThreatManager/StormVisuals/LargeClouds/T_StormCloudDetail01_M}
  n7 = TextureSample(UVs=Panner(Coordinate=TextureCoordinate() {uTiling=2.2, vTiling=2.2}, Time=@n5) {speedX=0.005, speedY=0.06}) {samplerType=SAMPLERTYPE_LinearGrayscale, texture=/Game/Environments/Hexmap/HexClouds/T_HexmapFog_01_Height02}
  n8 = Add(A=LinearInterpolate(A=@n6.R, B=@n7.R) {constAlpha=0.4}, B=Multiply(A=@n6.R, B=@n7.R))
  n9 = ScalarParameter() {defaultValue=0.5, parameterName=GlowAdd}
  n10 = Power(Base=Multiply(A=Power(Base=@n4) {constExponent=1.5}, B=LinearInterpolate(A=@n8, Alpha=@n9)))
  Multiply(A=Multiply(A=@n2.A, B=@n10))
  n11 = VectorParameter() {parameterName=DarkColor, defaultValue=<rgba 0.1215 0.1172 0.25 0>}
  n12 = VectorParameter() {parameterName=LightColor, defaultValue=<rgba 0.3016 0.4705 0.7 0>}
  Multiply(
   A=MaterialFunctionCall(
    Color_Morning=VectorParameter() {parameterName=MorningTint, defaultValue=<rgba 0.5 0.4842 0.4226 0>}.RGB
    Color_Day=VectorParameter() {parameterName=DayTint, defaultValue=<rgba 0.4656 0.5961 0.6 0>}.RGB
    Color_Evening=VectorParameter() {parameterName=EveningTint, defaultValue=<rgba 0.5 0.3191 0.2839 0>}.RGB
    Color_Night=VectorParameter() {parameterName=NightTint, defaultValue=<rgba 0.503 0.7265 1 0>}.RGB
   ) {materialFunction=&Color_Time_of_Day}.Result
   B=Multiply(A=@n2.RGB, B=LinearInterpolate(A=@n11.RGB, B=@n12.RGB, Alpha=@n10))
  )
 ]
}
