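# Additive unlit material for comet trail particles, on sprites or meshes, drawing a soft round puff of drifting fog and dust.
# A panning fog texture plus two squared, tinted dust layers are shaped by a radial sphere mask.
# A per-particle random dynamic parameter offsets panning time.
# The result blends dark to light color, multiplied by particle color and time-of-day tint, and combines with particle alpha for opacity.
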Material /Figment_S03/Effects/Fort_Effects/Materials/Comet/M_Comet_TrailParticle
graph {
 properties {
  bUsedWithMeshParticles=true
  bUsedWithParticleSprites=true
  blendMode=BLEND_Additive
  pixelDepthOffsetMode=PDOM_Legacy
  shadingModel=MSM_Unlit
  usedShadingModels=Unlit
 }
 nodes [
  Constant() {r=1}
  n1 = ParticleColor()
  n2 = Constant() {r=0.5}
  n3 = SphereMask(A=TextureCoordinate(), B=@n2, Radius=@n2) {attenuationRadius=0.5, hardnessPercent=25}
  n4 = TextureCoordinate() {uTiling=1.5, vTiling=1.5}
  n5 = Multiply(
   A=Add(A=Time(), B=DynamicParameter() {paramNames [Random,Param2,Param3,Param4]}.Random)
   B=ScalarParameter() {defaultValue=0.5, parameterName=PanSpeed_1}
  )
  n6 = Multiply(A=TextureSample(UVs=Panner(Coordinate=@n4, Time=@n5) {speedX=0.005, speedY=-0.01}) {samplerType=SAMPLERTYPE_LinearGrayscale, texture=/Game/Environments/Hexmap/HexClouds/T_HexmapFog_01_Height02}.R) {constB=0.5}
  n7 = TextureCoordinate() {uTiling=1.2, vTiling=1.2}
  n8 = TextureCoordinate() {uTiling=2.2, vTiling=2.2}
  n9 = Add(
   A=TextureSample(UVs=Panner(Coordinate=@n7, Time=@n5) {speedX=-0.004, speedY=0.001}) {texture=/Game/Global/Textures/Default/PointSplatter/T_Tiling_Dust}.RGB
   B=TextureSample(UVs=Panner(Coordinate=@n8, Time=@n5) {speedX=0.02, speedY=0.005}) {texture=&T_Tiling_Dust}.RGB
  )
  n10 = Multiply(A=Multiply(A=@n9, B=@n9), B=VectorParameter() {parameterName=DebrisColor, defaultValue=<rgba 7.5327 8.4067 16 0>}.RGB)
  n11 = Power(Base=Multiply(A=Power(Base=@n3) {constExponent=1.5}, B=Add(A=@n6, B=@n10)))
  Multiply(A=Multiply(A=@n1.A, B=@n11))
  n12 = VectorParameter() {parameterName=DarkColor, defaultValue=<rgba 0.1215 0.1172 0.25 0>}
  n13 = VectorParameter() {parameterName=LightColor, defaultValue=<rgba 0.3016 0.4705 0.7 0>}
  Multiply(
   A=MaterialFunctionCall(
    Color_Morning=VectorParameter() {parameterName=MorningTint, defaultValue=<rgba 0.5 0.4842 0.4226 0>}.RGB
    Color_Day=VectorParameter() {parameterName=DayTint, defaultValue=<rgba 0.3104 0.3974 0.4 0>}.RGB
    Color_Evening=VectorParameter() {parameterName=EveningTint, defaultValue=<rgba 0.5 0.3191 0.2839 0>}.RGB
    Color_Night=VectorParameter() {parameterName=NightTint, defaultValue=<rgba 0.503 0.7265 1 0>}.RGB
   ) {materialFunction=&Color_Time_of_Day}.Result
   B=Multiply(A=@n1.RGB, B=LinearInterpolate(A=@n12.RGB, B=@n13.RGB, Alpha=@n11))
  )
 ]
}
