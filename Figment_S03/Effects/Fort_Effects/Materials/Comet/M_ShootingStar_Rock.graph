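# Additive, unlit, two-sided particle material for the glowing rock of a shooting star or comet effect.
# Two panned noise textures blend between dark and light colors, with a V-gradient adding a bright hot end and fading toward a rear color.
# The result passes through an LDR-to-HDR remap, then is multiplied by particle color and a softened time-of-day tint.
# Particle alpha and a view-angle fade form a separate fade term.
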Material /Figment_S03/Effects/Fort_Effects/Materials/Comet/M_ShootingStar_Rock
graph {
 properties {
  bUsedWithMeshParticles=true
  bUsedWithParticleSprites=true
  blendMode=BLEND_Additive
  pixelDepthOffsetMode=PDOM_Legacy
  shadingModel=MSM_Unlit
  twoSided=true
  usedShadingModels=Unlit
 }
 nodes [
  n1 = VectorParameter() {parameterName=EveningTint, defaultValue=<rgba 0.8 0.6679 0.6042 0>}
  n2 = VectorParameter() {parameterName=DayTint, defaultValue=<rgba 0.4323 1 0.8228 0>}
  n3 = VectorParameter() {parameterName=NightTint, defaultValue=<rgba 0.304 0.2213 1.2 0>}
  n4 = ParticleColor()
  n5 = TextureCoordinate()
  n6 = Power(Base=OneMinus(None=ComponentMask(None=@n5) {g=true})) {constExponent=8}
  n7 = VectorParameter() {parameterName=LightColor, defaultValue=<rgba 3 1.6921 0.9062 0>}
  n8 = VectorParameter() {parameterName=DarkColor, defaultValue=<rgba 1 0.2427 0.2135 0>}
  n9 = TextureCoordinate() {uTiling=2}
  n10 = Multiply(A=Time(), B=ScalarParameter() {defaultValue=7, parameterName=PanSpeed})
  n11 = TextureCoordinate() {uTiling=2}
  n12 = LinearInterpolate(
   A=TextureSample(UVs=Panner(Coordinate=@n9, Time=@n10) {speedX=0.04, speedY=-0.18}) {samplerType=SAMPLERTYPE_LinearGrayscale, texture=/Game/Global/Textures/Default/Noise/T_SwirlNoise_N_height}.R
   B=TextureSample(UVs=Panner(Coordinate=@n11, Time=@n10) {speedX=-0.05, speedY=-0.25}) {texture=/Game/Effects/Fort_Effects/Textures/T_Smooth_tile}.R
  ) {constAlpha=0.4}
  n13 = Power(Base=ComponentMask(None=TextureCoordinate()) {g=true})
  n14 = LinearInterpolate(
   A=Add(A=Multiply(A=@n6, B=@n7.RGB), B=LinearInterpolate(A=@n8.RGB, B=@n7.RGB, Alpha=@n12))
   B=VectorParameter() {parameterName=RearColor, defaultValue=<rgba 0.5659 0.5354 0.6719 0>}.RGB
   Alpha=Clamp(None=Multiply(A=@n13) {constB=1.5})
  )
  n15 = Clamp(None=@n14)
  n16 = Multiply(A=@n14, B=ScalarParameter() {defaultValue=1, parameterName=HDRMult})
  Multiply(
   A=LinearInterpolate(A=MaterialFunctionCall(Color_Morning=@n1.RGB, Color_Day=@n2.RGB, Color_Evening=@n1.RGB, Color_Night=@n3.RGB) {materialFunction=&Color_Time_of_Day}.Result) {constAlpha=0.4}
   B=Multiply(
    A=@n4.RGB
    B=MaterialFunctionCall(LDR Input=@n15, HDR Tint=@n16) {materialFunction=&DeriveHDRfromLDR}.RemappedRange
   )
  )
  n17 = MaterialFunctionCall("FadeSensitivity (optional)"=ScalarParameter() {defaultValue=3, parameterName=FadeAngle}) {materialFunction=&FadeBasedOnViewAngle}
  Multiply(A=Multiply(A=@n4.A), B=Clamp(None=Multiply(A=@n17.Result) {constB=3}))
 ]
}
